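# Translucent UI material for the Oak portrait in the challenges UI, showing a tinted crop of a BarbequeLarry texture.
# Scaled and offset UVs select a half-height vertical band of the texture.
# A powered vertical UV gradient, scaled by Tint_Amount, blends the lower part of the image toward Tint_Color, orange by default.
# The texture's saturated alpha supplies the transparency.
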Material /Game/Athena/UI/Challenges/Art/Oak/M_UI_OakPortrait
graph {
 properties {
  bUsedWithStaticMesh=false
  blendMode=BLEND_Translucent
  materialDomain=MD_UI
  pixelDepthOffsetMode=PDOM_Legacy
  usedShadingModels=Default Lit
 }
 nodes [
  n1 = TextureCoordinate()
  n2 = AppendVector(A=Constant() {r=1}, B=Constant() {r=0.5})
  n3 = Constant()
  n4 = Constant() {r=0.175}
  n5 = TextureSample(UVs=Add(A=Multiply(A=@n1, B=@n2), B=AppendVector(A=@n3, B=@n4))) {texture=&/BarbequeLarryM_pickaxes}
  n6 = Constant() {r=-0.6}
  n7 = Power(Base=ComponentMask(None=Add(A=@n1, B=@n6)) {g=true})
  n8 = ScalarParameter() {defaultValue=1, parameterName=Tint_Amount}
  LinearInterpolate(A=@n5.RGB, B=VectorParameter() {parameterName=Tint_Color, defaultValue=<rgba 1 0.5833 0 1>}.RGB, Alpha=Saturate(None=Multiply(A=@n7, B=@n8)))
  Saturate(None=@n5.A)
 ]
}
